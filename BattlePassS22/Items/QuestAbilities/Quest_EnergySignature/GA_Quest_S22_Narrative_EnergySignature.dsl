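; Season 22 narrative quest ability that turns the player's backpack into a scanner for energy signature actors.
; On non-server instances, it gathers tracked actors, updates that list through gameplay events, and refreshes the backpack on a half-second looping timer.
; The nearest actor's facing angle drives a material dial and green-light flag, while distance sets blink speed on a squared curve.
; Lights and target position feed the backpack's idle VFX parameters.
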(event Ability|EventActivateAbility
  (|CallParentFunction)
  (if (not (Networking|IsServer))
    (bind _returnvalue (Utilities|Time|SetTimerbyEvent (EventDispatchers|CreateEvent :!event UpdateBackpack) 0.5 true))
    (Variables|Default|SetUpdateTimer _returnvalue)
    (bind (_asynctaskproxy _payload) (Ability|Tasks|WaitGameplayEvent "(TagName=\"Event.Quest.TrackQuestActor.Add.S22.EnergySignature\")")
      (:then
        (bind (_asynctaskproxy_2 _payload_2) (Ability|Tasks|WaitGameplayEvent "(TagName=\"Event.Quest.TrackQuestActor.Remove.S22.EnergySignature\")")
          (:then
            (bind _outactors (Actor|GetAllActorsOfClass (Variables|Default|GetTrackedActorClass)))
            (Variables|Default|SetTrackedActors _outactors))
          (:EventReceived
            (Utilities|Array|RemoveItem (Variables|Default|GetTrackedActors) (.Target (Utilities|Struct|BreakGameplayEventData _payload_2))))))
      (:EventReceived
        (Utilities|Array|AddUnique (Variables|Default|GetTrackedActors) (.Target (Utilities|Struct|BreakGameplayEventData _payload)))))))

(event Ability|EventOnEndAbility
  (params (bWasCancelled :type Boolean))
  (|CallParentFunction bWasCancelled)
  (Utilities|Time|ClearandInvalidateTimerbyHandle (Variables|Default|GetUpdateTimer)))

(event Ability|EventFailedToActivatePassiveAbility
  (params (ActorInfo :type Gameplay%20Ability%20Actor%20Info%20Structure))
  (|CallParentFunction))

(fn GetClosestGoalObject
  (bind _minactorinstance (LocalVariables|GetMinActorInstance))
  (bind _asfort_player_state (Utilities|Casting|CastToFortPlayerState (Ability|GetOwningActorfromActorInfo)))
  (for _array_element (Variables|Default|GetTrackedActors)
    (Utilities|IsValid _array_element
      (:Is%20Valid
        (if (or (not (Utilities|IsValid (LocalVariables|GetMinActorInstance) :!class /Script/Engine.KismetSystemLibrary)) (< (Default|GetDistanceSquaredBetweenActors _array_element (PlayerState|GetPawn _asfort_player_state)) (LocalVariables|GetMinDistanceSq)))
          (LocalVariables|SetMinActorInstance _array_element)
          (LocalVariables|SetMinDistanceSq (Default|GetDistanceSquaredBetweenActors _array_element (PlayerState|GetPawn _asfort_player_state)))))))
  (Utilities|IsValid _minactorinstance
    (:Is%20Valid
      (return _minactorinstance (Math|Vector|SelectVector (- (Transformation|GetActorLocation _minactorinstance) (Transformation|GetActorLocation (PlayerState|GetPawn _asfort_player_state))) "0, 0, 0" (Utilities|IsValid _minactorinstance :!class /Script/Engine.KismetSystemLibrary))))
    (:Is%20Not%20Valid
      (return 0 "0, 0, 0"))))

(fn UpdateBackpack
  (bind (_closestobject _directionvector) (CallFunction|GetClosestGoalObject))
  (Utilities|IsValid _closestobject
    (:Is%20Valid
      (bind _asfort_player_state (Utilities|Casting|CastToFortPlayerState (Ability|GetOwningActorfromActorInfo)))
      (bind _output_get (Variables|Default|SetScannerNeedle_Position (- 1.0 (/ (Class|AbilityFunctions|GetAngleBetweenTwoVectors (Math|Vector|GetForwardVector (Transformation|GetActorRotation (PlayerState|GetPawn _asfort_player_state))) _directionvector) 1.8e+02))))
      (Variables|Default|SetbGreenLightEnabled (>= _output_get 0.9))
      (bind _output_get_2 (Variables|Default|SetBlueLight_BlinkingSpeed (Math|Float|Lerp (Variables|Defaults|GetMinRadiusBlinkSpeed) (Variables|Defaults|GetMaxRadiusBlinkSpeed) (Math|Float|Power (- 1.0 (Math|Float|Clamp(Float) (/ (Math|Float|Max(Float) (- (Math|Vector|VectorLength _directionvector) (Variables|Defaults|GetMinRadius))) (- (Variables|Defaults|GetMaxRadius) (Variables|Defaults|GetMinRadius))))) 2.0)))) ; TODO For particle system call
    )
    (:Is%20Not%20Valid
      (Variables|Default|SetScannerNeedle_Position 0.0)
      (Variables|Default|SetBlueLight_BlinkingSpeed 0.0) ; TODO For particle system call
    ))
  ;[1 1 1] Setting FX parameters
  (bind _asplayer_pawn_athena (Utilities|Casting|CastToPlayerPawn_Athena (Ability|GetActivatingPawn)))
  (Effects|Components|ParticleSystem|SetFloatParameter (Class|CustomCharacterPartModifier|GetIdleVFXComponent (FortPawn|GetModifierforPartType _asplayer_pawn_athena "Backpack")) "BlinkSpeed" (Variables|Default|GetBlueLight_BlinkingSpeed))
  (Effects|Components|ParticleSystem|SetBoolParameter (Class|CustomCharacterPartModifier|GetIdleVFXComponent (FortPawn|GetModifierforPartType _asplayer_pawn_athena "Backpack")) "bGreenLightEnabled" (Variables|Default|GetbGreenLightEnabled))
  ;[1 1 1] Material Parameters
  (Rendering|Material|SetScalarParameterValueonMaterials (CustomCharacter|GetSkeletalMeshForPartOnPlayerPawn (FortPawn|GetModifierforPartType _asplayer_pawn_athena "Backpack")) "Dial" (Variables|Default|GetScannerNeedle_Position))
  (Utilities|IsValid _closestobject
    (:Is%20Valid
      ;[1 1 1] Temporarily disabling for testing purposes -Bruno
      (Effects|Components|ParticleSystem|SetVectorParameter (Class|CustomCharacterPartModifier|GetIdleVFXComponent (FortPawn|GetModifierforPartType _asplayer_pawn_athena "Backpack")) "ClosestObjectPosition" (Transformation|GetActorLocation _closestobject)))
    (:Is%20Not%20Valid
      _))
  (Effects|Components|ParticleSystem|SetFloatParameter (Class|CustomCharacterPartModifier|GetIdleVFXComponent (FortPawn|GetModifierforPartType _asplayer_pawn_athena "Backpack")) "GreenLightToggleDistance" (Variables|Defaults|GetMinRadius)))
